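# Parent material for the FuseBrookRune cosmetic shoe skin, adding an animated, running-reactive rune glow over the shared character and shoe functions.
# Mask channels and Fresnel drive time-scrolled gradient lookups, blended between normal and reactive ramps and tinted green, red and orange.
# Reactivity follows timed heat-up and cool-down ramps, replaced by a looping sine pulse in the front end.
# ReactiveON gates the surface and emissive overrides before sock, PostFX and hologram functions.
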Material /CosmeticShoes/Assets/FORT_Shoes/Shoe_YukonGladRusset/Skins/FuseBrookRune/Materials/M_ShoesParent_FuseBrook
graph {
 properties {
  bUseMaterialAttributes=true
  bUsedWithClothing=true
  bUsedWithMeshDeformer=true
  bUsedWithMorphTargets=true
  bUsedWithNanite=true
  bUsedWithNiagaraMeshParticles=true
  bUsedWithSkeletalMesh=true
  pixelDepthOffsetMode=PDOM_Legacy
  shadingModel=MSM_FromMaterialExpression
  usedShadingModels=Default Lit | Subsurface | Clear Coat | Subsurface Profile | Cloth
 }
 nodes [
  n1 = MaterialFunctionCall() {materialFunction=&MF_CharacterParent_2023_PreFX}
  Add(A=ChannelMaskParameter(None=@n1.Tech Art Mask) {maskChannel=Green, parameterName=HeatScanChannel, defaultValue=<rgba 0 1 0 0>}.R)
  n2 = GetMaterialAttributes(None=MaterialFunctionCall!(Material=@n1.Result) {materialFunction=/Game/Global/Materials/Cosmetics/Kicks/Shoes/Materials/MF_ShoeReactivityTest}.Result) {attributeGetTypes [
    #1
    #2
    #3
    #4
   ]}
  n3 = ScalarParameter() {parameterName=BasecolorMulti}
  n4 = ScalarParameter() {parameterName=ReactiveON}
  n5 = Fresnel() {baseReflectFraction=0, exponent=3}
  n6 = Multiply(A=@n5) {constB=2}
  n7 = TextureSampleParameter2D() {
   parameterName=TechArtMaskRE
   samplerType=SAMPLERTYPE_Masks
   texture=/CosmeticShoes/Assets/FORT_Shoes/Shoe_YukonGladRusset/Skins/FuseBrookRune/Textures/T_rune_Shoe_FX
  }
  n8 = NamedRerouteDeclaration(None=@n7.G) {name=RedArea, nodeColor=<rgba 0.2891 0 1 1>}
  n9 = Fresnel() {baseReflectFraction=0, exponent=1}
  n10 = Saturate(None=Add(A=@n9) {constB=0.2})
  n11 = SmoothStep(Value=NamedRerouteUsage()) {constMax=0.5}
  n12 = LinearInterpolate(Alpha=OneMinus(None=@n9)) {constA=0.8}
  n13 = Saturate(None=LinearInterpolate(A=Multiply(A=@n8, B=@n10), Alpha=Multiply(A=@n11, B=@n12)) {constB=0.8})
  n14 = Time()
  n15 = ScalarParameter() {defaultValue=0.3, parameterName=AnimateSpeed}
  n16 = AppendVector(A=Add(A=Add(A=@n6, B=@n13), B=Multiply(A=@n14, B=@n15)), B=Constant() {r=0.5})
  n17 = TextureSampleParameter2D(UVs=@n16) {
   group=Reactivity
   parameterName=Radial
   texture=/CosmeticShoes/Assets/FORT_Shoes/Shoe_YukonGladRusset/Skins/FuseBrookRune/Textures/T_RuneGradient01
  }
  n18 = TextureSampleParameter2D(UVs=@n16) {
   group=Reactivity
   parameterName=Radial_React
   texture=/CosmeticShoes/Assets/FORT_Shoes/Shoe_YukonGladRusset/Skins/FuseBrookRune/Textures/T_RuneGradient_React01
  }
  n19 = LinearInterpolate(A=@n17.RGB, B=@n18.RGB, Alpha=NamedRerouteUsage())
  n20 = VectorParameter() {parameterName=GreenTint, defaultValue=<rgba 0.0383 0.2274 0.1044 1>}
  n21 = Saturate(None=NamedRerouteUsage())
  n22 = NamedRerouteUsage()
  n23 = LinearInterpolate(A=Subtract(A=OneMinus(None=@n21), B=SmoothStep(Value=@n22) {constMax=0.4, constMin=0.2}), Alpha=NamedRerouteUsage()) {constB=0}
  n24 = NamedRerouteUsage()
  n25 = Saturate(None=Add(A=OneMinus(None=@n9)) {constB=0.2})
  n26 = NamedRerouteUsage()
  n27 = Multiply(A=NamedRerouteUsage(), B=LinearInterpolate(Alpha=Fresnel() {baseReflectFraction=0, exponent=1}) {constA=0.5}) {constB=0.5}
  n28 = LinearInterpolate(
   A=LinearInterpolate(
    A=LinearInterpolate(A=@n19, B=@n20.RGB, Alpha=@n23)
    B=VectorParameter() {parameterName=RedTint, defaultValue=<rgba 0.6781 0.0082 0.0626 1>}.RGB
    Alpha=Multiply(A=@n24, B=@n25)
   )
   B=VectorParameter() {parameterName=OrangeTint, defaultValue=<rgba 1 0.1946 0.0047 1>}.RGB
   Alpha=Saturate(None=Add(A=@n26, B=@n27))
  )
  n29 = SetMaterialAttributes!(
   MaterialAttributes=@n2.MaterialAttributes
   Base Color=LinearInterpolate(A=@n2.Base Color, B=Multiply(A=@n2.Base Color, B=@n3), Alpha=@n4)
   Emissive Color=LinearInterpolate(A=@n2.Emissive Color, B=Add(A=@n2.Emissive Color, B=@n28), Alpha=@n4)
   Specular=LinearInterpolate(A=@n2.Specular, Alpha=@n4) {constB=0.3}
   Roughness=LinearInterpolate(A=@n2.Roughness, Alpha=@n4) {constB=0.6}
  ) {attributeSetTypes [
    #1
    #2
    #3
    #4
   ]}
  MaterialFunctionCall!(In=MaterialFunctionCall(
    Material=MaterialFunctionCall!(InOpacity, Material=@n29, Mask=@n1.Tech Art Mask) {materialFunction=/Game/Global/Materials/Cosmetics/Kicks/Socks/Materials/MF_Sock}.Result
    M=@n1.M
    Pre Skinned Local Position=@n1.M
    Distance Blend=@n1.M
    Fresnel=@n1.M
   ) {materialFunction=&MF_CharacterParent_2023_PostFX}.Result) {materialFunction=/Game/Global/Materials/Cosmetics/Kicks/Shoes/Materials/MF_TechartShoeHologram}
  NamedRerouteDeclaration(None=@n7.R) {name=YellowLines, nodeColor=<rgba 1 0 0.4922 1>}
  NamedRerouteDeclaration(None=@n7.B) {name=YellowArea, nodeColor=<rgba 1 0 0.0937 1>}
  NamedRerouteDeclaration(None=@n5) {name=FresnelImpact, nodeColor=<rgba 1 0 0.9141 1>}
  n30 = Time()
  n31 = Subtract(A=@n30, B=ScalarParameter() {group=Reactivity, parameterName=RunningTimeStart})
  n32 = ScalarParameter() {defaultValue=3.2, group=Reactivity, parameterName=RunningCoolTimeDuration}
  n33 = OneMinus(None=Saturate(None=Divide(A=@n31, B=@n32)))
  n34 = ScalarParameter() {parameterName=StateChangeTime}
  n35 = Saturate(None=Divide(A=Subtract(A=@n30, B=@n34), B=ScalarParameter() {defaultValue=1.2, group=Reactivity, parameterName=RunningHeatTimeDuration}))
  n36 = Divide(A=Time()) {constB=10}
  n37 = Sine(None=Divide(A=Frac(None=@n36)))
  NamedRerouteDeclaration(None=LinearInterpolate(A=Multiply(A=@n33, B=@n35), B=SmoothStep(Value=@n37) {constMin=0.7}, Alpha=ScalarParameter() {group=Reactivity, parameterName=IsInFrontEnd})) {name=ReactiveAlpha, nodeColor=<rgba 0 0.8125 1 1>}
 ]
}
